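# Sound cue for the VividRazor CoffeePuff fight intro pose, built from that character's melee attack kick and swing sounds.
# A mixer layers one random kick, boosted in the mix, with one random swing, each through its own modulator, with the swing pitched down.
# It uses the melee weapon sound class, the shared attack attenuation asset and a concurrency override that limits simultaneous instances.
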SoundCue /VividRazor_CoffeePuff/Audio/IP_SC_VividRazor_CoffeePuff_Fight_Intro_Pose
root=Mixer_1
nodes [
 {id=Mixer_1, children [Modulator_5,Modulator_6], properties {inputVolume [1.25,1]}}
 {id=Modulator_5, children [Random_2], properties {volumeMax=1}}
 {id=Modulator_6, children [Random_3], properties {pitchMax=0.85, pitchMin=0.8, volumeMax=1, volumeMin=1}}
 {
  id=Random_2
  children [WavePlayer_6, WavePlayer_7, WavePlayer_12, WavePlayer_13]
  properties {bShouldExcludeFromBranchCulling=true, weights [1,1,1,1]}
 }
 {
  id=Random_3
  children [WavePlayer_14, WavePlayer_15, WavePlayer_18, WavePlayer_19, WavePlayer_20, WavePlayer_21]
  properties {bShouldExcludeFromBranchCulling=true, weights [1,1,1,1,1,1]}
 }
 {id=WavePlayer_12, properties {soundWaveAssetPtr=&/IP_VividRazor_CoffeePuff_Attack_Kick_03}}
 {id=WavePlayer_13, properties {soundWaveAssetPtr=&/IP_VividRazor_CoffeePuff_Attack_Kick_04}}
 {id=WavePlayer_14, properties {soundWaveAssetPtr=&/IP_VividRazor_CoffeePuff_Attack_Swing_01}}
 {id=WavePlayer_15, properties {soundWaveAssetPtr=&/IP_VividRazor_CoffeePuff_Attack_Swing_02}}
 {id=WavePlayer_18, properties {soundWaveAssetPtr=&/IP_VividRazor_CoffeePuff_Attack_Swing_03}}
 {id=WavePlayer_19, properties {soundWaveAssetPtr=&/IP_VividRazor_CoffeePuff_Attack_Swing_04}}
 {id=WavePlayer_20, properties {soundWaveAssetPtr=&/IP_VividRazor_CoffeePuff_Attack_Swing_05}}
 {id=WavePlayer_21, properties {soundWaveAssetPtr=&/IP_VividRazor_CoffeePuff_Attack_Swing_06}}
 {id=WavePlayer_6, properties {soundWaveAssetPtr=&/IP_VividRazor_CoffeePuff_Attack_Kick_01}}
 {id=WavePlayer_7, properties {soundWaveAssetPtr=&/IP_VividRazor_CoffeePuff_Attack_Kick_02}}
]
properties {
 attenuationSettings=&/IP_ATT_VividRazor_CoffeePuff_Attack_Default
 bDebug=true
 bOverrideConcurrency=true
 maxDistance=11890.5
 soundClassObject=&Fort_WeaponMelee
 volumeMultiplier=0.6
 attenuationOverrides {distanceAlgorithm=Logarithmic}
 concurrencyOverrides {maxCount=3}
}
