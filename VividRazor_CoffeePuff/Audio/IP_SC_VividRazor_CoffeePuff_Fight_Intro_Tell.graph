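# Sound cue for the VividRazor CoffeePuff fight intro tell, playing a single fire-named wave through one Wave Player with no randomization or mixing.
# It borrows the AOE Slam attack's attenuation asset and routes to the melee weapon sound class at slightly reduced volume.
# A concurrency override caps simultaneous instances at a low count, and the cue's debug flag is left enabled.
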SoundCue /VividRazor_CoffeePuff/Audio/IP_SC_VividRazor_CoffeePuff_Fight_Intro_Tell
root=WavePlayer_0
nodes [{id=WavePlayer_0, properties {soundWaveAssetPtr=&/IP_VividRazor_CoffeePuff_Fight_Intro_Tell_Fire_01}}]
properties {
 attenuationSettings=&/IP_ATT_VividRazor_CoffeePuff_Attack_AOE_Slam
 bDebug=true
 bOverrideConcurrency=true
 maxDistance=15854
 soundClassObject=&Fort_WeaponMelee
 volumeMultiplier=0.85
 attenuationOverrides {distanceAlgorithm=Logarithmic}
 concurrencyOverrides {maxCount=2}
}
