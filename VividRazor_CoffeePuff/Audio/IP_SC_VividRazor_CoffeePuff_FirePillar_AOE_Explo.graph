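# Sound cue for the explosion of the VividRazor CoffeePuff fire pillar area-of-effect attack.
# A single wave player feeds a modulator that applies slight random pitch variation at constant volume, with a small overall volume boost on the cue.
# It borrows the AOE slam attack attenuation asset and routes to the melee weapon sound class.
# A concurrency override caps simultaneous instances at three, and the debug flag is left on.
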SoundCue /VividRazor_CoffeePuff/Audio/IP_SC_VividRazor_CoffeePuff_FirePillar_AOE_Explo
root=Modulator_0
nodes [
 {id=Modulator_0, children [WavePlayer_0], properties {pitchMax=1.1, pitchMin=0.9, volumeMax=1, volumeMin=1}}
 {id=WavePlayer_0, properties {soundWaveAssetPtr=&/IP_VividRazor_CoffeePuff_FirePillar_AOE_Explo_01}}
]
properties {
 attenuationSettings=&/IP_ATT_VividRazor_CoffeePuff_Attack_AOE_Slam
 bDebug=true
 bOverrideConcurrency=true
 maxDistance=15854
 soundClassObject=&Fort_WeaponMelee
 volumeMultiplier=1.2
 attenuationOverrides {distanceAlgorithm=Logarithmic}
 concurrencyOverrides {maxCount=3}
}
