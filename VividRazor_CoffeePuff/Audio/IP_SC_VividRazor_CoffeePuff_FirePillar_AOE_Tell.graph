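# Sound cue for the warning tell of the VividRazor CoffeePuff fire pillar area-of-effect attack.
# A single wave player feeds the root directly, with a volume boost and the melee weapon sound class.
# The shared default attack attenuation asset sets spatial falloff, while a local concurrency override caps simultaneous instances.
# The debug flag is left enabled on the cue.
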SoundCue /VividRazor_CoffeePuff/Audio/IP_SC_VividRazor_CoffeePuff_FirePillar_AOE_Tell
root=WavePlayer_0
nodes [{id=WavePlayer_0, properties {soundWaveAssetPtr=&/IP_VividRazor_CoffeePuff_FirePillar_AOE_Tell_01}}]
properties {
 attenuationSettings=&/IP_ATT_VividRazor_CoffeePuff_Attack_Default
 bDebug=true
 bOverrideConcurrency=true
 maxDistance=11890.5
 soundClassObject=&Fort_WeaponMelee
 volumeMultiplier=1.35
 attenuationOverrides {distanceAlgorithm=Logarithmic}
 concurrencyOverrides {maxCount=3}
}
