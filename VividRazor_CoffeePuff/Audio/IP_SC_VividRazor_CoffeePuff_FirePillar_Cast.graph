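# Sound cue for the CoffeePuff fire pillar cast in VividRazor, reusing the character's attack kick and swing sounds.
# A mixer layers a random kick, boosted in the mix, over a random swing pitched down, each through its own modulator.
# It uses the melee weapon sound class, shared attack attenuation settings and a concurrency override capping simultaneous plays.
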SoundCue /VividRazor_CoffeePuff/Audio/IP_SC_VividRazor_CoffeePuff_FirePillar_Cast
root=Mixer_1
nodes [
 {id=Mixer_1, children [Modulator_1,Modulator_6], properties {inputVolume [1.25,1]}}
 {id=Modulator_1, children [Random_1], properties {volumeMax=1}}
 {id=Modulator_6, children [Random_3], properties {pitchMax=0.85, pitchMin=0.8, volumeMax=1, volumeMin=1}}
 {
  id=Random_1
  children [WavePlayer_0,WavePlayer_7,WavePlayer_8,WavePlayer_9]
  properties {bShouldExcludeFromBranchCulling=true, weights [1,1,1,1]}
 }
 {
  id=Random_3
  children [WavePlayer_10, WavePlayer_11, WavePlayer_18, WavePlayer_19, WavePlayer_20, WavePlayer_21]
  properties {bShouldExcludeFromBranchCulling=true, weights [1,1,1,1,1,1]}
 }
 {id=WavePlayer_0, properties {soundWaveAssetPtr=&/IP_VividRazor_CoffeePuff_Attack_Kick_01}}
 {id=WavePlayer_10, properties {soundWaveAssetPtr=&/IP_VividRazor_CoffeePuff_Attack_Swing_01}}
 {id=WavePlayer_11, properties {soundWaveAssetPtr=&/IP_VividRazor_CoffeePuff_Attack_Swing_02}}
 {id=WavePlayer_18, properties {soundWaveAssetPtr=&/IP_VividRazor_CoffeePuff_Attack_Swing_03}}
 {id=WavePlayer_19, properties {soundWaveAssetPtr=&/IP_VividRazor_CoffeePuff_Attack_Swing_04}}
 {id=WavePlayer_20, properties {soundWaveAssetPtr=&/IP_VividRazor_CoffeePuff_Attack_Swing_05}}
 {id=WavePlayer_21, properties {soundWaveAssetPtr=&/IP_VividRazor_CoffeePuff_Attack_Swing_06}}
 {id=WavePlayer_7, properties {soundWaveAssetPtr=&/IP_VividRazor_CoffeePuff_Attack_Kick_02}}
 {id=WavePlayer_8, properties {soundWaveAssetPtr=&/IP_VividRazor_CoffeePuff_Attack_Kick_03}}
 {id=WavePlayer_9, properties {soundWaveAssetPtr=&/IP_VividRazor_CoffeePuff_Attack_Kick_04}}
]
properties {
 attenuationSettings=&/IP_ATT_VividRazor_CoffeePuff_Attack_Default
 bDebug=true
 bOverrideConcurrency=true
 maxDistance=11890.5
 soundClassObject=&Fort_WeaponMelee
 attenuationOverrides {distanceAlgorithm=Logarithmic}
 concurrencyOverrides {maxCount=3}
}
