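# Empty SoundCue named for the player tell of the CoffeePuff FirePillar, with no sound nodes or root, so it plays nothing as it stands.
# It carries only playback settings: the weapon melee sound class, a shared close-range stereo melee attack attenuation asset, and a concurrency override limiting simultaneous instances.
# The debug flag is left enabled.
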SoundCue /VividRazor_CoffeePuff/Audio/IP_SC_VividRazor_CoffeePuff_FirePillar_Player_Tell
root=null
nodes []
properties {
 attenuationSettings=/Game/Sounds/Attenuation/Fort_Weapon_Melee_Attack_Close_Range_Stereo_Att
 bDebug=true
 bOverrideConcurrency=true
 maxDistance=2097152
 soundClassObject=&Fort_WeaponMelee
 attenuationOverrides {distanceAlgorithm=Logarithmic}
 concurrencyOverrides {maxCount=3}
}
