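# Sound cue for the VividRazor CoffeePuff HFPunch AOE attack, layering a kick sound with a heavy fire whoosh.
# A mixer plays one randomly chosen variant from each layer at once, with modulator variation on the kick but fixed pitch and volume on the whoosh.
# It routes to the melee weapon sound class, uses the shared attack attenuation asset, and overrides concurrency to cap simultaneous instances.
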SoundCue /VividRazor_CoffeePuff/Audio/IP_SC_VividRazor_CoffeePuff_HFPunch_AOE
root=Mixer_0
nodes [
 {id=Mixer_0, children [Modulator_0,Modulator_2], properties {inputVolume [1,1]}}
 {id=Modulator_0, children [Random_2], properties {volumeMax=1}}
 {id=Modulator_2, children [Random_0], properties {pitchMax=1, pitchMin=1, volumeMax=1, volumeMin=1}}
 {
  id=Random_0
  children [WavePlayer_1,WavePlayer_2,WavePlayer_3,WavePlayer_4,WavePlayer_7]
  properties {bShouldExcludeFromBranchCulling=true, weights [1,1,1,1,1]}
 }
 {
  id=Random_2
  children [WavePlayer_5, WavePlayer_6, WavePlayer_12, WavePlayer_13]
  properties {bShouldExcludeFromBranchCulling=true, weights [1,1,1,1]}
 }
 {id=WavePlayer_1, properties {soundWaveAssetPtr=&/IP_VividRazor_CoffeePuff_Fire_Whoosh_Heavy_01}}
 {id=WavePlayer_12, properties {soundWaveAssetPtr=&/IP_VividRazor_CoffeePuff_Attack_Kick_03}}
 {id=WavePlayer_13, properties {soundWaveAssetPtr=&/IP_VividRazor_CoffeePuff_Attack_Kick_04}}
 {id=WavePlayer_2, properties {soundWaveAssetPtr=&/IP_VividRazor_CoffeePuff_Fire_Whoosh_Heavy_02}}
 {id=WavePlayer_3, properties {soundWaveAssetPtr=&/IP_VividRazor_CoffeePuff_Fire_Whoosh_Heavy_03}}
 {id=WavePlayer_4, properties {soundWaveAssetPtr=&/IP_VividRazor_CoffeePuff_Fire_Whoosh_Heavy_04}}
 {id=WavePlayer_5, properties {soundWaveAssetPtr=&/IP_VividRazor_CoffeePuff_Attack_Kick_01}}
 {id=WavePlayer_6, properties {soundWaveAssetPtr=&/IP_VividRazor_CoffeePuff_Attack_Kick_02}}
 {id=WavePlayer_7, properties {soundWaveAssetPtr=&/IP_VividRazor_CoffeePuff_Fire_Whoosh_Heavy_05}}
]
properties {
 attenuationSettings=&/IP_ATT_VividRazor_CoffeePuff_Attack_Default
 bDebug=true
 bOverrideConcurrency=true
 maxDistance=11890.5
 soundClassObject=&Fort_WeaponMelee
 volumeMultiplier=1.3
 attenuationOverrides {distanceAlgorithm=Logarithmic}
 concurrencyOverrides {maxCount=3}
}
