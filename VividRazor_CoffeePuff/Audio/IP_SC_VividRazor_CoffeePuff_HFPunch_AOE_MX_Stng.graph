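# Sound cue for the music stinger of the VividRazor CoffeePuff HFPunch area-of-effect attack, classed as a melee weapon sound.
# A single wave player feeds the output directly with the HFPunch outro AOE stinger wave.
# Distance falloff comes from a shared attack music-tell attenuation asset, and a concurrency override caps simultaneous instances.
# The cue has its debug flag enabled.
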SoundCue /VividRazor_CoffeePuff/Audio/IP_SC_VividRazor_CoffeePuff_HFPunch_AOE_MX_Stng
root=WavePlayer_0
nodes [{id=WavePlayer_0, properties {soundWaveAssetPtr=&/IP_VividRazor_CoffeePuff_HFPunch_Outro_AOE_MX_Stng}}]
properties {
 attenuationSettings=&/IP_ATT_VividRazor_CoffeePuff_Attack_MX_Tell
 bDebug=true
 bOverrideConcurrency=true
 maxDistance=11890.5
 soundClassObject=&Fort_WeaponMelee
 attenuationOverrides {distanceAlgorithm=Logarithmic}
 concurrencyOverrides {maxCount=3}
}
