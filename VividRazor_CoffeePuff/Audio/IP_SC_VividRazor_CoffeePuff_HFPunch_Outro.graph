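# Sound cue for the outro of VividRazor CoffeePuff's HFPunch, playing a heavy fire whoosh in the melee weapon sound class.
# A single Random node picks among five equally weighted wave variations and is excluded from branch culling.
# The cue uses the shared default attack attenuation asset, overrides concurrency to cap simultaneous instances, and boosts volume slightly.
# Its debug flag is left enabled.
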SoundCue /VividRazor_CoffeePuff/Audio/IP_SC_VividRazor_CoffeePuff_HFPunch_Outro
root=Random_0
nodes [
 {
  id=Random_0
  children [WavePlayer_2,WavePlayer_3,WavePlayer_4,WavePlayer_5,WavePlayer_6]
  properties {bShouldExcludeFromBranchCulling=true, weights [1,1,1,1,1]}
 }
 {id=WavePlayer_2, properties {soundWaveAssetPtr=&/IP_VividRazor_CoffeePuff_Fire_Whoosh_Heavy_01}}
 {id=WavePlayer_3, properties {soundWaveAssetPtr=&/IP_VividRazor_CoffeePuff_Fire_Whoosh_Heavy_02}}
 {id=WavePlayer_4, properties {soundWaveAssetPtr=&/IP_VividRazor_CoffeePuff_Fire_Whoosh_Heavy_03}}
 {id=WavePlayer_5, properties {soundWaveAssetPtr=&/IP_VividRazor_CoffeePuff_Fire_Whoosh_Heavy_04}}
 {id=WavePlayer_6, properties {soundWaveAssetPtr=&/IP_VividRazor_CoffeePuff_Fire_Whoosh_Heavy_05}}
]
properties {
 attenuationSettings=&/IP_ATT_VividRazor_CoffeePuff_Attack_Default
 bDebug=true
 bOverrideConcurrency=true
 maxDistance=11890.5
 soundClassObject=&Fort_WeaponMelee
 volumeMultiplier=1.3
 attenuationOverrides {distanceAlgorithm=Logarithmic}
 concurrencyOverrides {maxCount=3}
}
