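# Sound cue for the outro punch of the VividRazor CoffeePuff HFPunch melee attack.
# A single wave player feeds the root directly, with no randomization or modulation nodes.
# It routes through the Fort_WeaponMelee sound class at reduced volume and references the shared CoffeePuff attack attenuation asset.
# A concurrency override caps simultaneous instances, and the debug flag is left enabled.
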SoundCue /VividRazor_CoffeePuff/Audio/IP_SC_VividRazor_CoffeePuff_HFPunch_Outro_Punch
root=WavePlayer_1
nodes [{id=WavePlayer_1, properties {soundWaveAssetPtr=&/IP_VividRazor_CoffeePuff_HFPunch_Outro_Punch_01}}]
properties {
 attenuationSettings=&/IP_ATT_VividRazor_CoffeePuff_Attack_Default
 bDebug=true
 bOverrideConcurrency=true
 maxDistance=11890.5
 soundClassObject=&Fort_WeaponMelee
 volumeMultiplier=0.7
 attenuationOverrides {distanceAlgorithm=Logarithmic}
 concurrencyOverrides {maxCount=3}
}
