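# Sound cue for the VividRazor CoffeePuff hookshot fire, layering a cast sound with hand foley.
# A mixer blends two modulated random branches, one of cast variants and one of slightly quieter foley variants, both excluded from branch culling.
# It plays through the melee weapon sound class with hookshot-specific attenuation and a concurrency override that caps simultaneous instances.
# The cue's debug flag is on.
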SoundCue /VividRazor_CoffeePuff/Audio/IP_SC_VividRazor_CoffeePuff_Hookshot_Fire
root=Mixer_0
nodes [
 {id=Mixer_0, children [Modulator_0,Modulator_2], properties {inputVolume [0.8,1]}}
 {id=Modulator_0, children [Random_1]}
 {id=Modulator_2, children [Random_0]}
 {id=Random_0, children [WavePlayer_0,WavePlayer_1,WavePlayer_2], properties {bShouldExcludeFromBranchCulling=true, weights [1,1,1]}}
 {
  id=Random_1
  children [WavePlayer_4,WavePlayer_5,WavePlayer_6,WavePlayer_7]
  properties {bShouldExcludeFromBranchCulling=true, weights [1,1,1,1]}
 }
 {id=WavePlayer_0, properties {soundWaveAssetPtr=&/IP_VividRazor_CoffeePuff_Hookshot_Cast_01}}
 {id=WavePlayer_1, properties {soundWaveAssetPtr=&/IP_VividRazor_CoffeePuff_Hookshot_Cast_02}}
 {id=WavePlayer_2, properties {soundWaveAssetPtr=&/IP_VividRazor_CoffeePuff_Hookshot_Cast_03}}
 {id=WavePlayer_4, properties {soundWaveAssetPtr=&/IP_VividRazor_CoffeePuff_Hookshot_Cast_Hand_Foley_01}}
 {id=WavePlayer_5, properties {soundWaveAssetPtr=&/IP_VividRazor_CoffeePuff_Hookshot_Cast_Hand_Foley_02}}
 {id=WavePlayer_6, properties {soundWaveAssetPtr=&/IP_VividRazor_CoffeePuff_Hookshot_Cast_Hand_Foley_03}}
 {id=WavePlayer_7, properties {soundWaveAssetPtr=&/IP_VividRazor_CoffeePuff_Hookshot_Cast_Hand_Foley_04}}
]
properties {
 attenuationSettings=&/IP_ATT_VividRazor_CoffeePuff_Hookshot
 bDebug=true
 bOverrideConcurrency=true
 maxDistance=16000
 soundClassObject=&Fort_WeaponMelee
 volumeMultiplier=1.25
 concurrencyOverrides {maxCount=3}
}
