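# Sound cue for a music stinger, named for the firing of the VividRazor CoffeePuff hookshot.
# A single wave player plays a charge-tell stinger wave (variant B) in the melee weapon sound class, with attack-tell attenuation and a volume boost.
# It sends fully to a weapon volume sidechain bus, overrides concurrency to two instances, disables virtualization and has its debug flag on.
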SoundCue /VividRazor_CoffeePuff/Audio/IP_SC_VividRazor_CoffeePuff_Hookshot_Fire_MX_Stng
root=WavePlayer_0
nodes [{id=WavePlayer_0, properties {soundWaveAssetPtr=&/IP_VividRazor_CoffeePuff_Hookshot_Charge_Tell_Stng_B}}]
properties {
 attenuationSettings=&/IP_ATT_VividRazor_CoffeePuff_Attack_MX_Tell
 bDebug=true
 bOverrideConcurrency=true
 maxDistance=11890.5
 soundClassObject=&Fort_WeaponMelee
 virtualizationMode=Disabled
 volumeMultiplier=1.65
 busSends [{audioBus=&AB_Sidechain_FortWeapon_Vol, minSendLevel=0, sendLevel=1, soundSourceBus=null}]
 concurrencyOverrides {maxCount=2}
}
