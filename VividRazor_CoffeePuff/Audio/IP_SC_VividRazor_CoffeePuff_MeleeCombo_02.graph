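# Sound cue for the second MeleeCombo attack of VividRazor CoffeePuff, layering a kick sound with a swing sound.
# A Mixer plays one of four random kick waves, boosted in level, alongside one of six random swing waves.
# Each layer has its own Modulator, with the swing pitched lower and quieter.
# Playback uses the melee weapon sound class, close-range attenuation and a concurrency override capping simultaneous instances.
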SoundCue /VividRazor_CoffeePuff/Audio/IP_SC_VividRazor_CoffeePuff_MeleeCombo_02
root=Mixer_0
nodes [
 {id=Mixer_0, children [Modulator_0,Modulator_1,null], properties {inputVolume [1.25,1,1]}}
 {id=Modulator_0, children [Random_2], properties {volumeMax=1}}
 {id=Modulator_1, children [Random_3], properties {pitchMax=0.85, pitchMin=0.8, volumeMax=0.8, volumeMin=0.7}}
 {
  id=Random_2
  children [WavePlayer_5, WavePlayer_6, WavePlayer_12, WavePlayer_13]
  properties {bShouldExcludeFromBranchCulling=true, weights [1,1,1,1]}
 }
 {
  id=Random_3
  children [WavePlayer_14, WavePlayer_15, WavePlayer_16, WavePlayer_17, WavePlayer_18, WavePlayer_19]
  properties {bShouldExcludeFromBranchCulling=true, weights [1,1,1,1,1,1]}
 }
 {id=WavePlayer_12, properties {soundWaveAssetPtr=&/IP_VividRazor_CoffeePuff_Attack_Kick_03}}
 {id=WavePlayer_13, properties {soundWaveAssetPtr=&/IP_VividRazor_CoffeePuff_Attack_Kick_04}}
 {id=WavePlayer_14, properties {soundWaveAssetPtr=&/IP_VividRazor_CoffeePuff_Attack_Swing_01}}
 {id=WavePlayer_15, properties {soundWaveAssetPtr=&/IP_VividRazor_CoffeePuff_Attack_Swing_02}}
 {id=WavePlayer_16, properties {soundWaveAssetPtr=&/IP_VividRazor_CoffeePuff_Attack_Swing_03}}
 {id=WavePlayer_17, properties {soundWaveAssetPtr=&/IP_VividRazor_CoffeePuff_Attack_Swing_04}}
 {id=WavePlayer_18, properties {soundWaveAssetPtr=&/IP_VividRazor_CoffeePuff_Attack_Swing_05}}
 {id=WavePlayer_19, properties {soundWaveAssetPtr=&/IP_VividRazor_CoffeePuff_Attack_Swing_06}}
 {id=WavePlayer_5, properties {soundWaveAssetPtr=&/IP_VividRazor_CoffeePuff_Attack_Kick_01}}
 {id=WavePlayer_6, properties {soundWaveAssetPtr=&/IP_VividRazor_CoffeePuff_Attack_Kick_02}}
]
properties {
 attenuationSettings=/Game/Sounds/Attenuation/Fort_Weapon_Melee_Attack_Close_Range_Mono_Att
 bDebug=true
 bOverrideConcurrency=true
 maxDistance=10058.75
 pitchMultiplier=1.15
 soundClassObject=&Fort_WeaponMelee
 volumeMultiplier=0.6
 attenuationOverrides {distanceAlgorithm=Logarithmic}
 concurrencyOverrides {maxCount=3}
}
