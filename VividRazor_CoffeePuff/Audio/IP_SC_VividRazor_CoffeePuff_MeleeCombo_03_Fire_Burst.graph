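# Sound cue for the fire burst in the third VividRazor CoffeePuff melee combo attack, using close-range melee attenuation.
# A mixer layers randomly picked burst and start-scream variations with slight volume or pitch modulation.
# A quiet, pitched-down embers layer follows after a fixed delay and is wired only to the first two quality-level slots.
# Concurrency is overridden to cap simultaneous instances, and virtualization is disabled.
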SoundCue /VividRazor_CoffeePuff/Audio/IP_SC_VividRazor_CoffeePuff_MeleeCombo_03_Fire_Burst
root=Mixer_0
nodes [
 {id=Delay_0, children [Modulator_6], properties {delayMax=0.65, delayMin=0.65}}
 {id=Mixer_0, children [Modulator_0, Modulator_2, QualityLevel_0], properties {inputVolume [1,1,0.1]}}
 {id=Modulator_0, children [Random_6], properties {volumeMax=1, volumeMin=0.9}}
 {id=Modulator_2, children [Random_5], properties {pitchMax=1.2, pitchMin=0.9, volumeMax=1, volumeMin=0.8}}
 {id=Modulator_6, children [Random_4], properties {pitchMax=0.9, pitchMin=0.75, volumeMax=1, volumeMin=1}}
 {id=QualityLevel_0, children [Delay_0,Delay_0,null,null]}
 {id=Random_4, children [WavePlayer_17, WavePlayer_18, WavePlayer_19], properties {bShouldExcludeFromBranchCulling=true, weights [1,1,1]}}
 {id=Random_5, children [WavePlayer_14, WavePlayer_15, WavePlayer_16], properties {bShouldExcludeFromBranchCulling=true, weights [1,1,1]}}
 {id=Random_6, children [WavePlayer_11, WavePlayer_12, WavePlayer_13], properties {bShouldExcludeFromBranchCulling=true, weights [1,1,1]}}
 {id=WavePlayer_11, properties {soundWaveAssetPtr=&/IP_VividRazor_CoffeePuff_Attack_03_Fire_Burst_01}}
 {id=WavePlayer_12, properties {soundWaveAssetPtr=&/IP_VividRazor_CoffeePuff_Attack_03_Fire_Burst_02}}
 {id=WavePlayer_13, properties {soundWaveAssetPtr=&/IP_VividRazor_CoffeePuff_Attack_03_Fire_Burst_03}}
 {id=WavePlayer_14, properties {soundWaveAssetPtr=&/IP_VividRazor_CoffeePuff_Attack_03_Fire_Start_Scream_01}}
 {id=WavePlayer_15, properties {soundWaveAssetPtr=&/IP_VividRazor_CoffeePuff_Attack_03_Fire_Start_Scream_02}}
 {id=WavePlayer_16, properties {soundWaveAssetPtr=&/IP_VividRazor_CoffeePuff_Attack_03_Fire_Start_Scream_03}}
 {id=WavePlayer_17, properties {soundWaveAssetPtr=&/IP_VividRazor_CoffeePuff_Attack_03_Fire_Burst_Embers_01}}
 {id=WavePlayer_18, properties {soundWaveAssetPtr=&/IP_VividRazor_CoffeePuff_Attack_03_Fire_Burst_Embers_02}}
 {id=WavePlayer_19, properties {soundWaveAssetPtr=&/IP_VividRazor_CoffeePuff_Attack_03_Fire_Burst_Embers_03}}
]
properties {
 attenuationSettings=/Game/Sounds/Attenuation/Fort_Weapon_Melee_Attack_Close_Range_Stereo_Att
 bOverrideConcurrency=true
 maxDistance=10058.75
 soundClassObject=&Fort_WeaponMelee
 virtualizationMode=Disabled
 volumeMultiplier=0.5
 concurrencyOverrides {maxCount=3}
}
