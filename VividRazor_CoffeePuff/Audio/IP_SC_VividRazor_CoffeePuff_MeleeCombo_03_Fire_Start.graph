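# Sound cue for the start of the third melee combo attack's fire sound on the VividRazor CoffeePuff weapon.
# A Random node picks one of three equally weighted attack waves and feeds it through a Modulator for variation.
# It uses the melee weapon sound class and close-range stereo melee attenuation, with virtualization disabled and a concurrency override capping simultaneous instances.
# The debug flag is left on.
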SoundCue /VividRazor_CoffeePuff/Audio/IP_SC_VividRazor_CoffeePuff_MeleeCombo_03_Fire_Start
root=Modulator_0
nodes [
 {id=Modulator_0, children [Random_1], properties {volumeMax=1}}
 {id=Random_1, children [WavePlayer_3,WavePlayer_4,WavePlayer_5], properties {bShouldExcludeFromBranchCulling=true, weights [1,1,1]}}
 {id=WavePlayer_3, properties {soundWaveAssetPtr=&/IP_VividRazor_CoffeePuff_Attack_03_Fire_Start_01}}
 {id=WavePlayer_4, properties {soundWaveAssetPtr=&/IP_VividRazor_CoffeePuff_Attack_03_Fire_Start_02}}
 {id=WavePlayer_5, properties {soundWaveAssetPtr=&/IP_VividRazor_CoffeePuff_Attack_03_Fire_Start_03}}
]
properties {
 attenuationSettings=/Game/Sounds/Attenuation/Fort_Weapon_Melee_Attack_Close_Range_Stereo_Att
 bDebug=true
 bOverrideConcurrency=true
 maxDistance=10058.75
 soundClassObject=&Fort_WeaponMelee
 virtualizationMode=Disabled
 volumeMultiplier=1.25
 attenuationOverrides {distanceAlgorithm=Logarithmic}
 concurrencyOverrides {maxCount=2}
}
