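# Looping ambient cue for the VividRazor CoffeePuff interactable, using Fortnite's loot ambient sound class and concurrency set.
# A mixer layers a looping bed, shaped by a rising volume envelope, with a repeating pulse that has slight random volume variation.
# Each layer has its own Attenuation node.
# The pulse runs through a Quality Level node with only the first input connected, so other quality levels get just the bed.
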SoundCue /VividRazor_CoffeePuff/Audio/IP_SC_VividRazor_Interact_Ambient_Loop
root=Mixer_0
nodes [
 {
  id=Attenuation_0
  children [Enveloper_0]
  properties {attenuationSettings=&/IP_ATT_VividRazor_CoffeePuff_Interact_Ambient_Loop}
 }
 {
  id=Attenuation_1
  children [WavePlayer_3]
  properties {attenuationSettings=&IP_ATT_VividRazor_CoffeePuff_Interact_Ambient_Loop}
 }
 {id=Delay_0, children [Modulator_1]}
 {
  id=Enveloper_0
  children [WavePlayer_0]
  properties {pitchCurve {editorCurveData {keys [{time=0}]}}, volumeCurve {editorCurveData {keys [<at 0 0>,<at 1 1>]}}}
 }
 {id=Looping_0, children [Delay_0]}
 {id=Mixer_0, children [Modulator_0, QualityLevel_0], properties {inputVolume [1,1.5]}}
 {id=Modulator_0, children [Attenuation_0], properties {pitchMax=1, pitchMin=1, volumeMax=1, volumeMin=1}}
 {id=Modulator_1, children [Attenuation_1], properties {pitchMax=1, pitchMin=1, volumeMax=1, volumeMin=0.8}}
 {id=QualityLevel_0, children [Looping_0,null,null,null]}
 {id=WavePlayer_0, properties {bLooping=true, soundWaveAssetPtr=&/IP_VividRazor_CoffeePuff_Interact_Ambient_Loop}}
 {id=WavePlayer_3, properties {soundWaveAssetPtr=&/IP_VividRazor_CoffeePuff_Interact_Ambient_Pulse}}
]
properties {
 maxDistance=2500
 priority=50
 soundClassObject=/Game/Sounds/FortSoundClassesAndModes/Fort_Loot_Ambient_Loop
 volumeMultiplier=1
 attenuationOverrides {
  bEnableFocusInterpolation=true
  bEnableListenerFocus=true
  distanceAlgorithm=Logarithmic
  falloffDistance=1100
  focusAttackInterpSpeed=20
  focusAzimuth=20
  focusReleaseInterpSpeed=10
  focusVolumeAttenuation=3.5
  lPFFrequencyAtMax=10000
  lPFRadiusMax=1000
  lPFRadiusMin=500
  nonFocusAzimuth=40
  nonFocusVolumeAttenuation=0.8
  nonSpatializedRadiusStart=200
  occlusionInterpolationTime=0.3
  occlusionLowPassFilterFrequency=3000
  occlusionVolumeAttenuation=0.75
  spatializationAlgorithm=SPATIALIZATION_HRTF
  attenuationShapeExtents {x=0}
  customAttenuationCurve {editorCurveData {keys [<at 0 1 RCIM_Cubic>, {arriveTangent=-1, interpMode=RCIM_Cubic, leaveTangent=-1, time=0.5, value=0.1}, <at 1 0 RCIM_Cubic>]}}
 }
 concurrencyOverrides {resolutionRule=StopQuietest}
 concurrencySet [/Game/Sounds/Concurrency_Rules/Loot_Ambient_Conc]
}
