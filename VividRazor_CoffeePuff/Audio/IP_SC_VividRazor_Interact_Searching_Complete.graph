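# Sound cue for the completion of a search interaction in the VividRazor CoffeePuff content, playing a single sweetener wave at boosted volume.
# It uses its own completion attenuation and the boss stinger music sound class, and sends to a music volume sidechain bus.
# A concurrency override allows one instance per owner and blocks new plays, superseding the referenced Dojo gong concurrency set.
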SoundCue /VividRazor_CoffeePuff/Audio/IP_SC_VividRazor_Interact_Searching_Complete
root=WavePlayer_2
nodes [{id=WavePlayer_2, properties {soundWaveAssetPtr=&/IP_VividRazor_CoffeePuff_Interact_Sweetener_01}}]
properties {
 attenuationSettings=&/IP_ATT_VividRazor_CoffeePuff_Interact_Complete
 bOverrideConcurrency=true
 maxDistance=18000
 soundClassObject=/Game/Sounds/FortSoundClassesAndModes/Fort_Music_Boss_Stinger
 volumeMultiplier=3
 busSends [{audioBus=&AB_Sidechain_FortMusic_Vol, minSendLevel=0, sendLevel=0.3, soundSourceBus=null}]
 concurrencyOverrides {bLimitToOwner=true, maxCount=1, resolutionRule=PreventNew, volumeScaleAttackTime=1}
 concurrencySet [/Game/Sounds/Asteria/Environment_Actors/Gong/Dojo_Gong_Conc]
}
