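# Sound cue for the looping "searching" interaction audio of VividRazor CoffeePuff.
# A looping wave player feeds an Enveloper that fades the volume in over the first half second.
# It plays through the Boss Totem stereo interact attenuation and the loot collect sound class, with boosted volume.
# A concurrency override limits it to a single instance per owner.
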SoundCue /VividRazor_CoffeePuff/Audio/IP_SC_VividRazor_Interact_Searching_Loop
root=Enveloper_0
nodes [
 {
  id=Enveloper_0
  children [WavePlayer_0]
  properties {pitchCurve {editorCurveData {keys [{time=0}]}}, volumeCurve {editorCurveData {keys [<at 0 0 RCIM_Cubic>,<at 0.5 1>]}}}
 }
 {id=WavePlayer_0, properties {bLooping=true, soundWaveAssetPtr=&/IP_VividRazor_CoffeePuff_Interact_Searching_Loop}}
]
properties {
 attenuationSettings=/Game/Sounds/Attenuation/Fort_Item_Interact_Boss_Totem_Stereo_Att
 bOverrideConcurrency=true
 maxDistance=12000
 soundClassObject=&Fort_Loot_Collect
 volumeMultiplier=2.5
 concurrencyOverrides {bLimitToOwner=true, maxCount=1, volumeScaleAttackTime=1}
}
