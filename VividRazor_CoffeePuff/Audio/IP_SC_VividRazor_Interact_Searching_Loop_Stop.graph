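# Sound cue that plays the stop sound for the searching interaction loop in VividRazor CoffeePuff.
# A single wave plays through a modulator that slightly varies pitch downward at random.
# It uses a boss totem item-interact attenuation and the loot collect sound class, with boosted volume.
# An overridden concurrency caps simultaneous instances per owner.
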SoundCue /VividRazor_CoffeePuff/Audio/IP_SC_VividRazor_Interact_Searching_Loop_Stop
root=Modulator_0
nodes [
 {id=Modulator_0, children [WavePlayer_1], properties {pitchMax=1, pitchMin=0.9, volumeMax=1, volumeMin=1}}
 {id=WavePlayer_1, properties {soundWaveAssetPtr=&/IP_VividRazor_CoffeePuff_Interact_Searching_Loop_Stop}}
]
properties {
 attenuationSettings=/Game/Sounds/Attenuation/Fort_Item_Interact_Boss_Totem_Stereo_Att
 bOverrideConcurrency=true
 maxDistance=12000
 soundClassObject=&Fort_Loot_Collect
 volumeMultiplier=2
 concurrencyOverrides {bLimitToOwner=true, maxCount=2, volumeScaleAttackTime=1}
}
