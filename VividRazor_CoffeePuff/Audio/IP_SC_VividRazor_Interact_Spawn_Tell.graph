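# Sound cue for the VividRazor CoffeePuff interact spawn tell, layering a fire pillar AOE tell with an interact spawn sound.
# A mixer plays both waves together, each through an attenuation node using the CoffeePuff spawn attenuation, with the spawn layer boosted.
# A concurrency override allows one instance and blocks new ones while it plays.
# It is routed to the boss music stinger sound class and has debug enabled.
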SoundCue /VividRazor_CoffeePuff/Audio/IP_SC_VividRazor_Interact_Spawn_Tell
root=Mixer_0
nodes [
 {
  id=Attenuation_0
  children [WavePlayer_0]
  properties {attenuationSettings=&/IP_ATT_VividRazor_CoffeePuff_Interact_Spawn}
 }
 {
  id=Attenuation_1
  children [WavePlayer_1]
  properties {attenuationSettings=&IP_ATT_VividRazor_CoffeePuff_Interact_Spawn}
 }
 {id=Mixer_0, children [Attenuation_0, Attenuation_1], properties {inputVolume [1,2.5]}}
 {id=WavePlayer_0, properties {soundWaveAssetPtr=&/IP_VividRazor_CoffeePuff_FirePillar_AOE_Tell_01}}
 {id=WavePlayer_1, properties {soundWaveAssetPtr=&/IP_VividRazor_CoffeePuff_Interact_Spawn_A}}
]
properties {
 bDebug=true
 bOverrideConcurrency=true
 maxDistance=16000
 soundClassObject=/Game/Sounds/FortSoundClassesAndModes/Fort_Music_Boss_Stinger
 volumeMultiplier=1
 attenuationOverrides {distanceAlgorithm=Logarithmic}
 concurrencyOverrides {maxCount=1, resolutionRule=PreventNew}
}
