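# Sound cue for the VividRazor CoffeePuff melee weapon's impact sound, named for hits on players.
# A Random node picks one of four equally weighted impact waves, is excluded from branch culling, and feeds a Modulator left at default settings.
# The cue plays at reduced volume in the melee weapon sound class, using close-range mono attenuation and a concurrency override that caps simultaneous instances at two.
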SoundCue /VividRazor_CoffeePuff/Audio/IP_SC_VividRazor_Melee_Impact_Player
root=Modulator_1
nodes [
 {id=Modulator_1, children [Random_0]}
 {
  id=Random_0
  children [WavePlayer_1,WavePlayer_2,WavePlayer_3,WavePlayer_4]
  properties {bShouldExcludeFromBranchCulling=true, weights [1,1,1,1]}
 }
 {id=WavePlayer_1, properties {soundWaveAssetPtr=&/IP_VividRazor_CoffeePuff_impact_Player_01}}
 {id=WavePlayer_2, properties {soundWaveAssetPtr=&/IP_VividRazor_CoffeePuff_impact_Player_02}}
 {id=WavePlayer_3, properties {soundWaveAssetPtr=&/IP_VividRazor_CoffeePuff_impact_Player_03}}
 {id=WavePlayer_4, properties {soundWaveAssetPtr=&/IP_VividRazor_CoffeePuff_impact_Player_04}}
]
properties {
 attenuationSettings=/Game/Sounds/Attenuation/Fort_Weapon_Melee_Attack_Close_Range_Mono_Att
 bOverrideConcurrency=true
 maxDistance=10058.75
 soundClassObject=&Fort_WeaponMelee
 volumeMultiplier=0.3
 concurrencyOverrides {maxCount=2}
}
